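# First-person fire sound cue for the Midas Drum Gun, using first-person weapon attenuation, sound class and gunfire concurrency.
# A mixer layers a random close-fire shot, a random "Ghost" layer at slightly reduced volume with upward pitch variation, and a low-ammo branch.
# The branch adds a low-ammo rifle sound only when the LowAmmo parameter is true.
# Both random pools are excluded from branch culling.
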SoundCue /Game/Sounds/Fort_Weapons/AutoRifle/DrumGun/SC_Weapon_DrumgunMidas_Fire_1P_Cue_Remade
root=Mixer_0
nodes [
 {id=Branch_11, children [WavePlayer_34, null, null], properties {boolParameterName=LowAmmo}}
 {id=Mixer_0, children [Random_4,Branch_11,Modulator_0], properties {inputVolume [1,1,0.7]}}
 {id=Modulator_0, children [Random_0], properties {pitchMax=1.4, pitchMin=1}}
 {
  id=Random_0
  children [WavePlayer_5, WavePlayer_6, WavePlayer_7, WavePlayer_11, WavePlayer_12]
  properties {bShouldExcludeFromBranchCulling=true, weights [1,1,1,1,1]}
 }
 {
  id=Random_4
  children [WavePlayer_8, WavePlayer_9, WavePlayer_10, WavePlayer_17]
  properties {bShouldExcludeFromBranchCulling=true, weights [1,1,1,1]}
 }
 {id=WavePlayer_10, properties {soundWaveAssetPtr=&/Drumgun_Fire_Close_03}}
 {id=WavePlayer_11, properties {soundWaveAssetPtr=&/DrumGunGhost_04_Remade}}
 {id=WavePlayer_12, properties {soundWaveAssetPtr=&/DrumGunGhost_05_Remade}}
 {id=WavePlayer_17, properties {soundWaveAssetPtr=&/Drumgun_Fire_Close_05}}
 {id=WavePlayer_34, properties {soundWaveAssetPtr=/Game/Sounds/Fort_Weapons/LowAmmo/weapon_low_ammo_hi_rifle_05}}
 {id=WavePlayer_5, properties {soundWaveAssetPtr=&/DrumGunGhost_01_Remade}}
 {id=WavePlayer_6, properties {soundWaveAssetPtr=&/DrumGunGhost_02_Remade}}
 {id=WavePlayer_7, properties {soundWaveAssetPtr=&/DrumGunGhost_03_Remade}}
 {id=WavePlayer_8, properties {soundWaveAssetPtr=&/Drumgun_Fire_Close_01}}
 {id=WavePlayer_9, properties {soundWaveAssetPtr=&/Drumgun_Fire_Close_02}}
]
properties {
 attenuationSettings=&Fort_Weapon_Projectile_FirstPerson_Athena_Att
 bDebug=true
 maxDistance=2097152
 soundClassObject=&Fort_WeaponRanged_Fire_1P_PSM
 volumeMultiplier=1
 concurrencyOverrides {bLimitToOwner=true, maxCount=3, resolutionRule=StopOldest}
 concurrencySet [&Weapon_Gunfire_FirstPerson_Conc]
}
