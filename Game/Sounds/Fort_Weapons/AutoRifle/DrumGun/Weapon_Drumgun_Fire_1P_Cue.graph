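# First-person fire sound cue for the Drum Gun auto rifle.
# A mixer layers a random pick from four close-range fire waves with a branch that adds a low-ammo rifle sound only when the LowAmmo parameter is true.
# The random node is excluded from branch culling.
# It uses first-person weapon attenuation, sound class and gunfire concurrency assets, alongside owner-limited, stop-oldest override settings.
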SoundCue /Game/Sounds/Fort_Weapons/AutoRifle/DrumGun/Weapon_Drumgun_Fire_1P_Cue
root=Mixer_0
nodes [
 {id=Branch_11, children [WavePlayer_34, null, null], properties {boolParameterName=LowAmmo}}
 {id=Mixer_0, children [Random_4,Branch_11], properties {inputVolume [1,1]}}
 {
  id=Random_4
  children [WavePlayer_8, WavePlayer_9, WavePlayer_10, WavePlayer_17]
  properties {bShouldExcludeFromBranchCulling=true, weights [1,1,1,1]}
 }
 {id=WavePlayer_10, properties {soundWaveAssetPtr=&/Drumgun_Fire_Close_03}}
 {id=WavePlayer_17, properties {soundWaveAssetPtr=&/Drumgun_Fire_Close_05}}
 {id=WavePlayer_34, properties {soundWaveAssetPtr=/Game/Sounds/Fort_Weapons/LowAmmo/weapon_low_ammo_hi_rifle_05}}
 {id=WavePlayer_8, properties {soundWaveAssetPtr=&/Drumgun_Fire_Close_01}}
 {id=WavePlayer_9, properties {soundWaveAssetPtr=&/Drumgun_Fire_Close_02}}
]
properties {
 attenuationSettings=&Fort_Weapon_Projectile_FirstPerson_Athena_Att
 bDebug=true
 maxDistance=2097152
 soundClassObject=&Fort_WeaponRanged_Fire_1P_PSM
 volumeMultiplier=1.15
 concurrencyOverrides {bLimitToOwner=true, maxCount=3, resolutionRule=StopOldest}
 concurrencySet [&Weapon_Gunfire_FirstPerson_Conc]
}
