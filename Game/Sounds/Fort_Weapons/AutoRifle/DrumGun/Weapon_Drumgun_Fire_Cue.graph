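# Drum Gun fire sound cue that layers close, medium and distant gunshots, each through its own distance attenuation.
# The close layer picks one of four waves, overrides stereo spread and uses the third-person ranged-fire sound class.
# A quieter submix under the distant-fire class blends three random, pitch-raised medium waves with one distant wave.
# Local and global gunfire concurrency sets limit playback, and the debug flag is on.
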SoundCue /Game/Sounds/Fort_Weapons/AutoRifle/DrumGun/Weapon_Drumgun_Fire_Cue
root=Mixer_1
nodes [
 {
  id=Attenuation_118
  children [Random_4]
  properties {attenuationSettings=&Fort_Weapon_Projectile_Close_Athena_Att, attenuationOverrides {stereoSpread=29871.3984}}
 }
 {
  id=Attenuation_119
  children [Modulator_2]
  properties {attenuationSettings=&Fort_Weapon_Projectile_Medium_Mono_Athena_Att}
 }
 {
  id=Attenuation_120
  children [Modulator_1]
  properties {attenuationSettings=&Fort_Weapon_Projectile_Distant_Mono_Athena_Att}
 }
 {id=Mixer_0, children [Attenuation_119, Attenuation_120], properties {inputVolume [0.325,0.125]}}
 {id=Mixer_1, children [SoundClass_77, SoundClass_76], properties {inputVolume [1,1]}}
 {id=Modulator_1, children [WavePlayer_23], properties {pitchMin=1, volumeMax=1, volumeMin=1}}
 {id=Modulator_2, children [Random_5], properties {pitchMax=1.2, pitchMin=1.1, volumeMax=2, volumeMin=2}}
 {
  id=Random_4
  children [WavePlayer_8, WavePlayer_9, WavePlayer_10, WavePlayer_17]
  properties {bShouldExcludeFromBranchCulling=true, weights [1,1,1,1]}
 }
 {id=Random_5, children [WavePlayer_20, WavePlayer_21, WavePlayer_22], properties {bShouldExcludeFromBranchCulling=true, weights [1,1,1]}}
 {id=SoundClass_76, children [Mixer_0], properties {soundClassOverride=&Fort_WeaponRanged_Fire_Distant}}
 {
  id=SoundClass_77
  children [Attenuation_118]
  properties {soundClassOverride=&Fort_WeaponRanged_Fire_3P_PSM}
 }
 {id=WavePlayer_10, properties {soundWaveAssetPtr=&/Drumgun_Fire_Close_03}}
 {id=WavePlayer_17, properties {soundWaveAssetPtr=&/Drumgun_Fire_Close_05}}
 {id=WavePlayer_20, properties {soundWaveAssetPtr=&/Drumgun_Fire_MediumNew_01}}
 {id=WavePlayer_21, properties {soundWaveAssetPtr=&/Drumgun_Fire_MediumNew_02}}
 {id=WavePlayer_22, properties {soundWaveAssetPtr=&/Drumgun_Fire_MediumNew_03}}
 {id=WavePlayer_23, properties {soundWaveAssetPtr=&/Drumgun_Fire_DistantNew_01}}
 {id=WavePlayer_8, properties {soundWaveAssetPtr=&/Drumgun_Fire_Close_01}}
 {id=WavePlayer_9, properties {soundWaveAssetPtr=&/Drumgun_Fire_Close_02}}
]
properties {
 bDebug=true
 maxDistance=27500
 volumeMultiplier=1
 concurrencySet [&Weapon_Gunfire_Local_Conc, &Weapon_Gunfire_Global_Conc]
}
