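# Sound cue named for the low-ammo sound of the Drum Gun auto rifle, with an empty node graph and no root node, so it references no audio.
# It carries a slightly boosted volume, a very large max distance and the degraded ranged-weapon low-ammo sound class.
# A concurrency override uses the PreventNew rule, refusing new instances once the limit is reached.
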SoundCue /Game/Sounds/Fort_Weapons/AutoRifle/DrumGun/Weapon_Drumgun_LowAmmo_Cue
root=null
nodes []
properties {
 bOverrideConcurrency=true
 maxDistance=2097152
 soundClassObject=/Game/Sounds/FortSoundClassesAndModes/Fort_WeaponRanged_LowAmmo_Degraded_PSM
 volumeMultiplier=1.2
 concurrencyOverrides {resolutionRule=PreventNew}
}
