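# Sound cue for the drum gun's reload insert sound, an auto rifle weapon foley effect.
# A single wave player feeds the Weapon_Drumgun_ReloadInsert wave straight to the output, with no randomization or modulation.
# It uses the shared weapon reload attenuation and the weapon foley sound class, with playback volume lowered.
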SoundCue /Game/Sounds/Fort_Weapons/AutoRifle/DrumGun/Weapon_Drumgun_ReloadInsert_Cue
root=WavePlayer_0
nodes [{id=WavePlayer_0, properties {soundWaveAssetPtr=&/Weapon_Drumgun_ReloadInsert}}]
properties {
 attenuationSettings=&Fort_Weapon_Reload_Att
 maxDistance=2250
 soundClassObject=&Fort_WeaponFoley
 volumeMultiplier=0.601
}
